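# Sound cue for the bounce of the beach ball toy in Athena.
# A Random node picks one of two equally weighted bounce waves and feeds a Modulator at default settings for slight pitch and volume variation.
# Output has boosted volume, uses the shared Athena ball-bounce attenuation, and plays through the Fort_Music_Emote sound class rather than an effects class.
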SoundCue /BRCosmetics/Athena/Sounds/Toys/Toys_BeachBall_Bounce
root=Modulator_1
nodes [
 {id=Modulator_1, children [Random_2]}
 {id=Random_2, children [WavePlayer_8,WavePlayer_9], properties {weights [1,1]}}
 {id=WavePlayer_8, properties {soundWaveAssetPtr=/Game/Athena/Sounds/Toys/Toys_BeachBall_Bounce_13}}
 {id=WavePlayer_9, properties {soundWaveAssetPtr=/Game/Athena/Sounds/Toys/Toys_BeachBall_Bounce_15}}
]
properties {
 attenuationSettings=/Game/Sounds/Attenuation/Fort_Ball_Bounce_Athena_Att
 maxDistance=7000
 soundClassObject=&Fort_Music_Emote
 volumeMultiplier=1.8
}
